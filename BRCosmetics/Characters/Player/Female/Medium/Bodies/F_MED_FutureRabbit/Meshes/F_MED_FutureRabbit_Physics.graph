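# Physics asset for the FutureRabbit medium female player body, giving collision and constraints for its dangling rabbit ears, shoelaces, bags and tail.
# Capsules and spheres cover torso, limbs and hands, while a huge root sphere sits below the origin like a floor.
# Limited-swing ear and lace chains hang from bases locked to spine_05, the right ear via dyn_hood_01.
# Bags rotate freely on the pelvis, whereas both tail joints lock rotation.
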PhysicsAsset /BRCosmetics/Characters/Player/Female/Medium/Bodies/F_MED_FutureRabbit/Meshes/F_MED_FutureRabbit_Physics
bodies {
 calf_l [{shapeType=Capsule, length=32.3923, radius=9.2145, center=<-27.1354 1.4225 -0.0986>, rotation=<rot 0 78.2758 -89.8901>}]
 calf_r [{shapeType=Capsule, length=32.3923, radius=9.2145, center=<27.1354 -1.4225 0.0986>, rotation=<rot 0 78.2758 90.1099>}]
 dyn_bag_l_base [{shapeType=Sphere, radius=6.7674, shapeName=dyn_bag_l_base_sphere, center=<-0.1621 -0.0917 -1.4323>}]
 dyn_bag_r_base [{shapeType=Sphere, radius=6.7674, shapeName=dyn_bag_r_base_sphere, center=<0.7775 -0.0677 -1.3483>}]
 dyn_ear_l_01 [{
   shapeType=Capsule
   length=6.463
   radius=3.2355
   shapeName=dyn_ear_l_01_capsule
   center=<5.2183 -1.406 0.0931>
   rotation=<rot -21.7907 94.8124 -90.8325>
  }]
 dyn_ear_l_02 [{
   shapeType=Capsule
   length=7.5778
   radius=3.1416
   shapeName=dyn_ear_l_02_capsule
   center=<4.0699 -0.3063 -0.1151>
   rotation=<rot 0 93.4478 -90.412>
  }]
 dyn_ear_l_03 [{
   shapeType=Capsule
   length=8.3717
   radius=3.1768
   shapeName=dyn_ear_l_03_capsule
   center=<3.5818 0.2827 -0.4756>
   rotation=<rot -0.6255 87.8218 -94.5807>
  }]
 dyn_ear_l_04 [{
   shapeType=Capsule
   length=8.57
   radius=4.8654
   shapeName=dyn_ear_l_04_capsule
   center=<4.4778 -2.3027 0.1475>
   rotation=<rot 0 93.0887 -90.5214>
  }]
 dyn_ear_l_05 [{
   shapeType=Capsule
   length=7.8907
   radius=4.8321
   shapeName=dyn_ear_l_05_capsule
   center=<5.5463 -1.8327 1.9643>
   rotation=<rot 0 95.7984 96.0378>
  }]
 dyn_ear_l_base [{shapeType=Capsule, length=0.505, radius=0.505, shapeName=dyn_ear_l_base_capsule, rotation=<rot 0 -89.9999 -6.0663>}]
 dyn_ear_r_01 [{
   shapeType=Capsule
   length=7.9836
   radius=3.412
   shapeName=dyn_ear_r_01_capsule
   center=<-5.5277 1.0733 0.2014>
   rotation=<rot -21.7926 95.428 -89.7831>
  }]
 dyn_ear_r_02 [{
   shapeType=Capsule
   length=8.2669
   radius=2.9057
   shapeName=dyn_ear_r_02_capsule
   center=<-5.6921 -0.167 0.5361>
   rotation=<rot 0 93.2627 -91.007>
  }]
 dyn_ear_r_03 [{
   shapeType=Capsule
   length=6.7057
   radius=4.8427
   shapeName=dyn_ear_r_03_capsule
   center=<-3.5818 -0.2827 0.4756>
   rotation=<rot -0.6255 87.8218 85.4193>
  }]
 dyn_ear_r_04 [{
   shapeType=Capsule
   length=7.82
   radius=5.6154
   shapeName=dyn_ear_r_04_capsule
   center=<-4.4386 3.1381 -0.9089>
   rotation=<rot 0 93.0888 96.3062>
  }]
 dyn_ear_r_05 [{
   shapeType=Capsule
   length=7.8907
   radius=4.8321
   shapeName=dyn_ear_r_05_capsule
   center=<-5.5131 1.5057 -1.9643>
   rotation=<rot 0 95.7984 -83.9622>
  }]
 dyn_hood_01 [{shapeType=Capsule, length=0.505, radius=0.505, shapeName=dyn_hood_01_capsule, center=<0.0262 0 -0.247>, rotation=<rot 0 -89.9999 -6.0663>}]
 dyn_shoelace_l_01 [{
   shapeType=Capsule
   length=3.6171
   radius=0.9227
   shapeName=dyn_shoelace_l_01_capsule
   center=<3.6361 -0.224 0.1645>
   rotation=<rot -0.0799 85.7386 -82.4486>
  }]
 dyn_shoelace_l_02 [{
   shapeType=Capsule
   length=5.6753
   radius=1.1867
   shapeName=dyn_shoelace_l_02_capsule
   center=<3.5174 -0.8047 0.6374>
   rotation=<rot 0 83.0906 -86.0816>
  }]
 dyn_shoelace_l_base [{shapeType=Capsule, length=0.505, radius=0.505, shapeName=dyn_shoelace_l_base_capsule, rotation=<rot 0 -90 0>}]
 dyn_shoelace_r_01 [{
   shapeType=Capsule
   length=4.2656
   radius=0.8837
   shapeName=dyn_shoelace_r_01_capsule
   center=<3.0352 0.1344 0.5888>
   rotation=<rot 0 -86.0877 -89.3995>
  }]
 dyn_shoelace_r_02 [{
   shapeType=Capsule
   length=6.5858
   radius=1.4443
   shapeName=dyn_shoelace_r_02_capsule
   center=<2.5531 0.1326 0.4572>
   rotation=<rot 0 86.4191 -87.2927>
  }]
 dyn_shoelace_r_base [{shapeType=Capsule, length=0.505, radius=0.505, shapeName=dyn_shoelace_r_base_capsule, rotation=<rot 0 -90 0>}]
 dyn_tail_01 [
  {shapeType=Sphere, radius=8.9122, shapeName=dyn_tail_01_sphere, center=<-2.1956 -6.6825 -0.0023>}
  {shapeType=Sphere, radius=9.1795, shapeName=dyn_tail_01_sphere}
 ]
 dyn_tail_base [{shapeType=Sphere, radius=0.505, shapeName=dyn_tail_base_sphere}]
 hand_l [{shapeType=Capsule, length=9.3968, radius=6.0864, center=<7.0823 -2.2915 0.4508>, rotation=<rot 0 83.1155 -71.9677>}]
 hand_r [{shapeType=Capsule, length=9.3968, radius=6.0864, center=<-7.0823 2.2915 -0.4508>, rotation=<rot 0 83.1155 -71.9677>}]
 pelvis [{shapeType=Capsule, length=14.408, radius=10.0436, center=<-3.3838 -1.9093 -0.057>, rotation=<rot -0.9322 -90 -0.3991>}]
 root [{shapeType=Sphere, radius=477.0089, shapeName=root_sphere, center=<0 0 -475>}]
 spine_04 [
  {shapeType=Sphere, radius=5.7943, shapeName=spine_04_sphere, center=<13.1703 0 -10.127>}
  {shapeType=Sphere, radius=6.3076, shapeName=spine_04_sphere, center=<13.2582 -1.0423 9.1637>}
  {shapeType=Capsule, length=9.8259, radius=10.3941, center=<-1.4162 0.625 -0.0757>, rotation=<rot 87.0285 77.5237 72.3218>}
  {shapeType=Capsule, length=13.0233, radius=5.4474, shapeName=spine_04_capsule, center=<3.8557 5.8349 0>}
  {shapeType=Capsule, length=15.4399, radius=4.5346, shapeName=spine_04_capsule, center=<6.7817 4.07 0>}
 ]
 spine_05 [{shapeType=Capsule, length=14.8124, radius=13.1708, center=<10.7287 -1.5067 -0.1587>, rotation=<rot 5.203 -90 3.8447>}]
 thigh_l [{shapeType=Capsule, length=24.1679, radius=8.5475, center=<-21.2966 -0.7032 -1.2741>, rotation=<rot 0 88.4212 -89.7182>}]
 thigh_r [{shapeType=Capsule, length=24.1679, radius=8.5475, center=<21.2966 0.7032 1.2741>, rotation=<rot 0 88.4212 90.2818>}]
 upperarm_l [{shapeType=Capsule, length=28.1978, radius=6.2241, center=<23.3592 3.2604 -0.6942>, rotation=<rot 0 99.7366 -86.8268>}]
 upperarm_r [{shapeType=Capsule, length=28.1978, radius=6.2241, center=<-23.3592 -3.2604 0.6942>, rotation=<rot 0 99.7366 93.1732>}]
}
constraints [
 {bone1Name=spine_04, bone2Name=pelvis, swing1LimitDegrees=45, swing2LimitDegrees=45, twistLimitDegrees=45}
 {bone1Name=spine_05, bone2Name=spine_04, swing1LimitDegrees=45, swing2LimitDegrees=45, twistLimitDegrees=45}
 {bone1Name=upperarm_l, bone2Name=spine_05, swing1LimitDegrees=45, swing2LimitDegrees=45, twistLimitDegrees=45}
 {bone1Name=hand_l, bone2Name=upperarm_l, swing1LimitDegrees=45, swing2LimitDegrees=45, twistLimitDegrees=45}
 {bone1Name=upperarm_r, bone2Name=spine_05, swing1LimitDegrees=45, swing2LimitDegrees=45, twistLimitDegrees=45}
 {bone1Name=hand_r, bone2Name=upperarm_r, swing1LimitDegrees=45, swing2LimitDegrees=45, twistLimitDegrees=45}
 {bone1Name=thigh_l, bone2Name=pelvis, swing1LimitDegrees=45, swing2LimitDegrees=45, twistLimitDegrees=45}
 {bone1Name=calf_l, bone2Name=thigh_l, swing1LimitDegrees=45, swing2LimitDegrees=45, twistLimitDegrees=45}
 {bone1Name=thigh_r, bone2Name=pelvis, swing1LimitDegrees=45, swing2LimitDegrees=45, twistLimitDegrees=45}
 {bone1Name=calf_r, bone2Name=thigh_r, swing1LimitDegrees=45, swing2LimitDegrees=45, twistLimitDegrees=45}
 {bone1Name=dyn_ear_r_05, bone2Name=dyn_ear_r_04, swing1LimitDegrees=27, swing2LimitDegrees=27, twistLimitDegrees=5}
 {
  bone1Name=dyn_ear_l_base
  bone2Name=spine_05
  swing1LimitDegrees=17
  swing1Motion=Locked
  swing2LimitDegrees=17
  swing2Motion=Locked
  twistLimitDegrees=7
  twistMotion=Locked
 }
 {bone1Name=dyn_ear_l_01, bone2Name=dyn_ear_l_base, swing1LimitDegrees=44, swing2LimitDegrees=17, twistLimitDegrees=5}
 {bone1Name=dyn_ear_l_02, bone2Name=dyn_ear_l_01, swing1LimitDegrees=17, swing2LimitDegrees=17, twistLimitDegrees=5}
 {bone1Name=dyn_ear_l_03, bone2Name=dyn_ear_l_02, swing1LimitDegrees=17, swing2LimitDegrees=17, twistLimitDegrees=5}
 {bone1Name=dyn_ear_l_04, bone2Name=dyn_ear_l_03, swing1LimitDegrees=27, swing2LimitDegrees=27, twistLimitDegrees=5}
 {bone1Name=dyn_ear_l_05, bone2Name=dyn_ear_l_04, swing1LimitDegrees=27, swing2LimitDegrees=27, twistLimitDegrees=5}
 {
  bone1Name=dyn_hood_01
  bone2Name=spine_05
  swing1LimitDegrees=45
  swing1Motion=Locked
  swing2LimitDegrees=45
  swing2Motion=Locked
  twistLimitDegrees=5
  twistMotion=Locked
 }
 {bone1Name=dyn_ear_r_01, bone2Name=dyn_hood_01, swing1LimitDegrees=44, swing2LimitDegrees=17, twistLimitDegrees=7}
 {bone1Name=dyn_ear_r_02, bone2Name=dyn_ear_r_01, swing1LimitDegrees=17, swing2LimitDegrees=17, twistLimitDegrees=5}
 {bone1Name=dyn_ear_r_03, bone2Name=dyn_ear_r_02, swing1LimitDegrees=17, swing2LimitDegrees=17, twistLimitDegrees=5}
 {bone1Name=dyn_ear_r_04, bone2Name=dyn_ear_r_03, swing1LimitDegrees=27, swing2LimitDegrees=27, twistLimitDegrees=5}
 {
  bone1Name=dyn_tail_base
  bone2Name=pelvis
  swing1LimitDegrees=45
  swing1Motion=Locked
  swing2LimitDegrees=45
  swing2Motion=Locked
  twistLimitDegrees=45
  twistMotion=Locked
 }
 {
  bone1Name=dyn_tail_01
  bone2Name=dyn_tail_base
  swing1LimitDegrees=45
  swing1Motion=Locked
  swing2LimitDegrees=45
  swing2Motion=Locked
  twistLimitDegrees=45
  twistMotion=Locked
 }
 {
  bone1Name=dyn_bag_l_base
  bone2Name=pelvis
  swing1LimitDegrees=45
  swing1Motion=Free
  swing2LimitDegrees=45
  swing2Motion=Free
  twistLimitDegrees=45
  twistMotion=Free
 }
 {
  bone1Name=dyn_bag_r_base
  bone2Name=pelvis
  swing1LimitDegrees=45
  swing1Motion=Free
  swing2LimitDegrees=45
  swing2Motion=Free
  twistLimitDegrees=45
  twistMotion=Free
 }
 {bone1Name=dyn_shoelace_l_01, bone2Name=dyn_shoelace_l_base, swing1LimitDegrees=19.808, swing2LimitDegrees=15.608, twistLimitDegrees=5}
 {bone1Name=dyn_shoelace_l_02, bone2Name=dyn_shoelace_l_01, swing1LimitDegrees=8.064, swing2LimitDegrees=11, twistLimitDegrees=5}
 {bone1Name=dyn_shoelace_r_01, bone2Name=dyn_shoelace_r_base, swing1LimitDegrees=44, swing2LimitDegrees=11, twistLimitDegrees=5}
 {bone1Name=dyn_shoelace_r_02, bone2Name=dyn_shoelace_r_01, swing1LimitDegrees=44, swing2LimitDegrees=11, twistLimitDegrees=5}
 {
  bone1Name=dyn_shoelace_l_base
  bone2Name=spine_05
  swing1LimitDegrees=45
  swing1Motion=Locked
  swing2LimitDegrees=45
  swing2Motion=Locked
  twistLimitDegrees=45
  twistMotion=Locked
 }
 {
  bone1Name=dyn_shoelace_r_base
  bone2Name=spine_05
  swing1LimitDegrees=45
  swing1Motion=Locked
  swing2LimitDegrees=45
  swing2Motion=Locked
  twistLimitDegrees=45
  twistMotion=Locked
 }
 {
  bone1Name=pelvis
  bone2Name=root
  swing1LimitDegrees=45
  swing1Motion=Free
  swing2LimitDegrees=45
  swing2Motion=Free
  twistLimitDegrees=45
  twistMotion=Free
 }
]
