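# Physics asset for the FutureRabbit female medium body, set up for secondary motion of its dyn_ rabbit ears and shoelaces.
# Capsules cover torso, arms, hands and legs, and a huge root sphere offset downward forms a near-flat floor.
# Five-segment ear chains use tight limits from bases locked to spine_05, the right through dyn_hood_01.
# Two-segment lace chains hang from bases free relative to spine_05, while the tail is locked to the pelvis.
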PhysicsAsset /BRCosmetics/Characters/Player/Female/Medium/Bodies/F_MED_FutureRabbit/Meshes/F_MED_FutureRabbit_Physics_Pack
bodies {
 calf_l [{shapeType=Capsule, length=32.3923, radius=9.2145, center=<-27.1354 1.4225 -0.0986>, rotation=<rot 0 78.2758 -89.8901>}]
 calf_r [{shapeType=Capsule, length=32.3923, radius=9.2145, center=<27.1354 -1.4225 0.0986>, rotation=<rot 0 78.2758 90.1099>}]
 dyn_ear_l_01 [{
   shapeType=Capsule
   length=6.463
   radius=3.2355
   shapeName=dyn_ear_l_01_capsule
   center=<5.2183 -1.406 0.0931>
   rotation=<rot -21.7907 94.8124 -90.8325>
  }]
 dyn_ear_l_02 [{
   shapeType=Capsule
   length=7.1691
   radius=3.5504
   shapeName=dyn_ear_l_02_capsule
   center=<4.0699 -0.3063 -0.1151>
   rotation=<rot 0 93.4478 -90.412>
  }]
 dyn_ear_l_03 [{
   shapeType=Capsule
   length=7.7132
   radius=3.8352
   shapeName=dyn_ear_l_03_capsule
   center=<3.5818 0.2827 -0.4756>
   rotation=<rot -0.6255 87.8218 -94.5807>
  }]
 dyn_ear_l_04 [{
   shapeType=Capsule
   length=8.57
   radius=4.8654
   shapeName=dyn_ear_l_04_capsule
   center=<4.4778 -2.3027 0.1475>
   rotation=<rot 0 93.0887 -90.5214>
  }]
 dyn_ear_l_05 [{
   shapeType=Capsule
   length=7.8907
   radius=4.8321
   shapeName=dyn_ear_l_05_capsule
   center=<5.5463 -1.8327 1.9643>
   rotation=<rot 0 95.7984 96.0378>
  }]
 dyn_ear_l_base [{shapeType=Capsule, length=0.505, radius=0.505, shapeName=dyn_ear_l_base_capsule, rotation=<rot 0 -89.9999 -6.0663>}]
 dyn_ear_r_01 [{
   shapeType=Capsule
   length=7.9836
   radius=3.412
   shapeName=dyn_ear_r_01_capsule
   center=<-5.5277 1.0733 0.2014>
   rotation=<rot -21.7926 95.428 -89.7831>
  }]
 dyn_ear_r_02 [{
   shapeType=Capsule
   length=7.5115
   radius=3.6611
   shapeName=dyn_ear_r_02_capsule
   center=<-5.6921 -0.167 0.5361>
   rotation=<rot 0 93.2627 -91.007>
  }]
 dyn_ear_r_03 [{
   shapeType=Capsule
   length=7.7132
   radius=3.8352
   shapeName=dyn_ear_r_03_capsule
   center=<-3.5818 -0.2827 0.4756>
   rotation=<rot -0.6254 87.8218 94.0363>
  }]
 dyn_ear_r_04 [{
   shapeType=Capsule
   length=7.07
   radius=6.3654
   shapeName=dyn_ear_r_04_capsule
   center=<-4.438 3.1275 -0.9089>
   rotation=<rot 0 93.0888 96.3062>
  }]
 dyn_ear_r_05 [{
   shapeType=Capsule
   length=7.8907
   radius=4.8321
   shapeName=dyn_ear_r_05_capsule
   center=<-5.5131 1.5057 -1.9643>
   rotation=<rot 0 95.7984 -83.9622>
  }]
 dyn_hood_01 [{shapeType=Capsule, length=0.505, radius=0.505, shapeName=dyn_hood_01_capsule, center=<0.0262 0 -0.247>, rotation=<rot 0 -89.9999 -6.0663>}]
 dyn_shoelace_l_01 [{
   shapeType=Capsule
   length=6.144
   radius=1.1768
   shapeName=dyn_shoelace_l_01_capsule
   center=<3.6954 0.11 0.4147>
   rotation=<rot 0 84.9018 -87.8172>
  }]
 dyn_shoelace_l_02 [{
   shapeType=Capsule
   length=5.21
   radius=1.1565
   shapeName=dyn_shoelace_l_02_capsule
   center=<1.05 -0.5567 0.4125>
   rotation=<rot 0 -101.8397 -91.5885>
  }]
 dyn_shoelace_l_base [{shapeType=Capsule, length=0.505, radius=0.505, shapeName=dyn_shoelace_l_base_capsule, rotation=<rot 0 -90 0>}]
 dyn_shoelace_r_01 [{
   shapeType=Capsule
   length=3.8749
   radius=1.1768
   shapeName=dyn_shoelace_r_01_capsule
   center=<3.6182 0.1348 0.0247>
   rotation=<rot 0 92.8751 -88.7862>
  }]
 dyn_shoelace_r_02 [{
   shapeType=Capsule
   length=2.5274
   radius=1.1565
   shapeName=dyn_shoelace_r_02_capsule
   center=<2.4875 -0.0054 0.5439>
   rotation=<rot 0 81.863 -84.2864>
  }]
 dyn_shoelace_r_base [{shapeType=Capsule, length=0.505, radius=0.505, shapeName=dyn_shoelace_r_base_capsule, rotation=<rot 0 -90 0>}]
 dyn_tail_01 [
  {shapeType=Sphere, radius=8.0395, shapeName=dyn_tail_01_sphere, center=<-2.1956 -6.6825 -0.0023>}
  {shapeType=Sphere, radius=7, shapeName=dyn_tail_01_sphere}
 ]
 dyn_tail_base [{shapeType=Sphere, radius=0.505, shapeName=dyn_tail_base_sphere}]
 hand_l [{shapeType=Capsule, length=9.3968, radius=6.0864, center=<7.0823 -2.2915 0.4508>, rotation=<rot 0 83.1155 -71.9677>}]
 hand_r [{shapeType=Capsule, length=9.3968, radius=6.0864, center=<-7.0823 2.2915 -0.4508>, rotation=<rot 0 83.1155 -71.9677>}]
 pelvis [{shapeType=Capsule, length=14.408, radius=10.0436, center=<-3.3838 -1.9093 -0.057>, rotation=<rot -0.9322 -90 -0.3991>}]
 root [{shapeType=Sphere, radius=477.0089, shapeName=root_sphere, center=<0 0 -475>}]
 spine_03 [{shapeType=Capsule, length=3.2841, radius=14.5446, center=<5.113 -2.0994 0>}]
 spine_04 [{shapeType=Capsule, length=12.3357, radius=12.0733, center=<-0.9525 1.3923 -0.1198>, rotation=<rot 87.0285 77.5237 72.3218>}]
 spine_05 [
  {shapeType=Capsule, length=25.7929, radius=7.4405, shapeName=spine_05_capsule, center=<-2.4975 1.9292 0.1257>, rotation=<rot 0 -90 0>}
  {shapeType=Capsule, length=22.3665, radius=7.6335, shapeName=spine_05_capsule, center=<0.2183 0 0>}
 ]
 thigh_l [{shapeType=Capsule, length=24.1679, radius=8.5475, center=<-21.2966 -0.7032 -1.2741>, rotation=<rot 0 88.4212 -89.7182>}]
 thigh_r [{shapeType=Capsule, length=24.1679, radius=8.5475, center=<21.2966 0.7032 1.2741>, rotation=<rot 0 88.4212 90.2818>}]
 upperarm_l [{shapeType=Capsule, length=28.1978, radius=6.2241, center=<23.3592 3.2604 -0.6942>, rotation=<rot 0 99.7366 -86.8268>}]
 upperarm_r [{shapeType=Capsule, length=28.1978, radius=6.2241, center=<-23.3592 -3.2604 0.6942>, rotation=<rot 0 99.7366 93.1732>}]
}
constraints [
 {bone1Name=spine_04, bone2Name=spine_03, swing1LimitDegrees=45, swing2LimitDegrees=45, twistLimitDegrees=45}
 {bone1Name=spine_05, bone2Name=spine_04, swing1LimitDegrees=45, swing2LimitDegrees=45, twistLimitDegrees=45}
 {bone1Name=upperarm_l, bone2Name=spine_05, swing1LimitDegrees=45, swing2LimitDegrees=45, twistLimitDegrees=45}
 {bone1Name=hand_l, bone2Name=upperarm_l, swing1LimitDegrees=45, swing2LimitDegrees=45, twistLimitDegrees=45}
 {bone1Name=upperarm_r, bone2Name=spine_05, swing1LimitDegrees=45, swing2LimitDegrees=45, twistLimitDegrees=45}
 {bone1Name=hand_r, bone2Name=upperarm_r, swing1LimitDegrees=45, swing2LimitDegrees=45, twistLimitDegrees=45}
 {bone1Name=thigh_l, bone2Name=pelvis, swing1LimitDegrees=45, swing2LimitDegrees=45, twistLimitDegrees=45}
 {bone1Name=calf_l, bone2Name=thigh_l, swing1LimitDegrees=45, swing2LimitDegrees=45, twistLimitDegrees=45}
 {bone1Name=thigh_r, bone2Name=pelvis, swing1LimitDegrees=45, swing2LimitDegrees=45, twistLimitDegrees=45}
 {bone1Name=calf_r, bone2Name=thigh_r, swing1LimitDegrees=45, swing2LimitDegrees=45, twistLimitDegrees=45}
 {bone1Name=dyn_ear_r_05, bone2Name=dyn_ear_r_04, swing1LimitDegrees=27, swing2LimitDegrees=27, twistLimitDegrees=5}
 {
  bone1Name=dyn_ear_l_base
  bone2Name=spine_05
  swing1LimitDegrees=17
  swing1Motion=Locked
  swing2LimitDegrees=17
  swing2Motion=Locked
  twistLimitDegrees=7
  twistMotion=Locked
 }
 {bone1Name=dyn_ear_l_01, bone2Name=dyn_ear_l_base, swing1LimitDegrees=1, swing2LimitDegrees=19, twistLimitDegrees=5}
 {bone1Name=dyn_ear_l_02, bone2Name=dyn_ear_l_01, swing1LimitDegrees=5, swing2LimitDegrees=17, twistLimitDegrees=5}
 {bone1Name=dyn_ear_l_03, bone2Name=dyn_ear_l_02, swing1LimitDegrees=17, swing1Motion=Locked, swing2LimitDegrees=17, twistLimitDegrees=5}
 {bone1Name=dyn_ear_l_04, bone2Name=dyn_ear_l_03, swing1LimitDegrees=35, swing2LimitDegrees=35, twistLimitDegrees=5}
 {bone1Name=dyn_ear_l_05, bone2Name=dyn_ear_l_04, swing1LimitDegrees=27, swing2LimitDegrees=27, twistLimitDegrees=5}
 {
  bone1Name=dyn_hood_01
  bone2Name=spine_05
  swing1LimitDegrees=45
  swing1Motion=Locked
  swing2LimitDegrees=45
  swing2Motion=Locked
  twistLimitDegrees=5
  twistMotion=Locked
 }
 {bone1Name=dyn_ear_r_01, bone2Name=dyn_hood_01, swing1LimitDegrees=1, swing2LimitDegrees=10.936, twistLimitDegrees=7}
 {bone1Name=dyn_ear_r_02, bone2Name=dyn_ear_r_01, swing1LimitDegrees=1, swing2LimitDegrees=10.088, twistLimitDegrees=5}
 {bone1Name=dyn_ear_r_03, bone2Name=dyn_ear_r_02, swing1LimitDegrees=17, swing1Motion=Locked, swing2LimitDegrees=10.088, twistLimitDegrees=5}
 {bone1Name=dyn_ear_r_04, bone2Name=dyn_ear_r_03, swing1LimitDegrees=35, swing2LimitDegrees=35, twistLimitDegrees=5}
 {
  bone1Name=dyn_tail_base
  bone2Name=pelvis
  swing1LimitDegrees=45
  swing1Motion=Locked
  swing2LimitDegrees=45
  swing2Motion=Locked
  twistLimitDegrees=45
  twistMotion=Locked
 }
 {
  bone1Name=dyn_tail_01
  bone2Name=dyn_tail_base
  swing1LimitDegrees=45
  swing1Motion=Locked
  swing2LimitDegrees=45
  swing2Motion=Locked
  twistLimitDegrees=45
  twistMotion=Locked
 }
 {
  bone1Name=pelvis
  bone2Name=root
  swing1LimitDegrees=45
  swing1Motion=Free
  swing2LimitDegrees=45
  swing2Motion=Free
  twistLimitDegrees=45
  twistMotion=Free
 }
 {
  bone1Name=spine_03
  bone2Name=pelvis
  swing1LimitDegrees=45
  swing1Motion=Free
  swing2LimitDegrees=45
  swing2Motion=Free
  twistLimitDegrees=45
  twistMotion=Free
 }
 {
  bone1Name=dyn_shoelace_l_base
  bone2Name=spine_05
  swing1LimitDegrees=45
  swing1Motion=Free
  swing2LimitDegrees=45
  swing2Motion=Free
  twistLimitDegrees=45
  twistMotion=Free
 }
 {bone1Name=dyn_shoelace_l_01, bone2Name=dyn_shoelace_l_base, swing1LimitDegrees=17, swing2LimitDegrees=11, twistLimitDegrees=5}
 {bone1Name=dyn_shoelace_l_02, bone2Name=dyn_shoelace_l_01, swing1LimitDegrees=17, swing2LimitDegrees=11, twistLimitDegrees=5}
 {
  bone1Name=dyn_shoelace_r_base
  bone2Name=spine_05
  swing1LimitDegrees=45
  swing1Motion=Free
  swing2LimitDegrees=45
  swing2Motion=Free
  twistLimitDegrees=45
  twistMotion=Free
 }
 {bone1Name=dyn_shoelace_r_01, bone2Name=dyn_shoelace_r_base, swing1LimitDegrees=17, swing2LimitDegrees=11, twistLimitDegrees=5}
 {bone1Name=dyn_shoelace_r_02, bone2Name=dyn_shoelace_r_01, swing1LimitDegrees=17, swing2LimitDegrees=11, twistLimitDegrees=5}
]
