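# Post-process style animation blueprint for the female FutureRabbit body cosmetic, adding secondary motion to a pose copied from the parent mesh.
# A Control Rig deforms the pose, then two RigidBody simulations using pack and non-pack physics assets blend by whether backbling is worn.
# Anim Dynamics springs the tail, whose base scales to zero during Cauldron and Scrub-a-Dub emotes.
# The event graph updates gravity override, backbling and emote flags.
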AnimBlueprint /BRCosmetics/Characters/Player/Female/Medium/Bodies/F_MED_FutureRabbit/Meshes/F_MED_FutureRabbit_AnimBP
target_skeleton=&/F_MED_FutureRabbit_Skeleton
graphs [
 AnimGraph {nodes [
   n1 = RigidBody!(
    ComponentPose=Local To Component(LocalPose=Use cached pose 'Deform_1'())
    Alpha
    OverrideWorldGravity:Vector=Get Gravity Override()
    ExternalForce
   ) -> (Pose) {
    bClampLinearTranslationLimitToRefPose=true
    bForceDisableCollisionBetweenConstraintBodies=true
    bOverrideWorldGravity=true
    overridePhysicsAsset=&/F_MED_FutureRabbit_Physics_Pack
    simulationSpace=BaseBoneSpace
    alphaBoolBlend {blendOption=Linear}
    baseBoneRef {boneName=spine_05}
    simSpaceSettings {
     dampingAlpha=0
     maxAngularAcceleration=40
     maxAngularVelocity=6
     maxLinearAcceleration=1500
     maxLinearVelocity=700
     velocityScaleZ=0.7
     worldAlpha=1
     externalLinearDragV {x=2}
    }
   }
   n2 = Component To Local(ComponentPose=@n1)
   n3 = RigidBody(ComponentPose=Local To Component(LocalPose=Use cached pose 'Deform_1'()), OverrideWorldGravity=Get Gravity Override()) {
    bClampLinearTranslationLimitToRefPose=true
    bForceDisableCollisionBetweenConstraintBodies=true
    bOverrideWorldGravity=true
    overridePhysicsAsset=&/F_MED_FutureRabbit_Physics
    simulationSpace=BaseBoneSpace
    alphaBoolBlend {blendOption=Linear}
    baseBoneRef {boneName=spine_05}
    simSpaceSettings {
     dampingAlpha=0
     maxAngularAcceleration=40
     maxAngularVelocity=6
     maxLinearAcceleration=1500
     maxLinearVelocity=700
     velocityScaleZ=0.7
     worldAlpha=1
     externalLinearDragV {x=2}
    }
   }
   n4 = Component To Local(ComponentPose=@n3)
   n5 = Get isWearingPack() -> (isWearingPack:bool)
   n6 = Anim Dynamics - Bone: dyn_tail_01(
    ComponentPose=Local To Component(LocalPose=Blend Poses by bool(BlendPose_0=@n2, BlendPose_1=@n4, bActiveValue=@n5) {blendType=Linear})
    Alpha=1
    ExternalForce="0,0,0"
   ) {
    angularDampingOverride=1
    angularSpringConstant=600
    bAngularSpring=true
    bOverrideAngularDamping=true
    bUseGravityOverride=true
    simulationSpace=BoneRelative
    alphaBoolBlend {blendOption=Linear}
    boundBone {boneName=dyn_tail_01}
    componentAppliedLinearAccClamp=<5000 5000 5000>
    componentLinearAccScale=<0.3 0.3 0.3>
    componentLinearVelScale=<0.3 0.3 0.3>
    gravityOverride {z=-980}
    physicsBodyDefinitions [{
      boundBone {boneName=dyn_tail_01}
      boxExtents=<12 12 12>
      constraintSetup {angularLimitsMax=<11 10 1>, angularLimitsMin=<-11 -10 -30>, angularTarget=<1 0 0>}
      localJointOffset=<1 0 0>
     }]
    relativeSpaceBone {boneName=root}
   }
   n7 = "Transform (Modify) Bone - Bone: dyn_tail_base"!(
    ComponentPose=@n6
    bAlphaBoolEnabled:bool="Get scale tail  cauldro"() -> ("scale tail  cauldro":bool)
    Translation="0,0,0"
    Rotation="0,0,0"
    Scale="0,0,0"
   ) -> (Pose) {alphaInputType=Bool, scaleMode=BMM_Additive, alphaBoolBlend {blendInTime=2}, boneToModify {boneName=dyn_tail_base}, scale=<0 0 0>}
   n8 = "Transform (Modify) Bone - Bone: dyn_tail_base"(
    ComponentPose=@n7
    bAlphaBoolEnabled="Get scale tail  scrubadub"() -> ("scale tail  scrubadub":bool)
    Translation="0,0,0"
    Rotation="0,0,0"
    Scale="0,0,0"
   ) {alphaInputType=Bool, scaleMode=BMM_Additive, alphaBoolBlend {blendInTime=1}, boneToModify {boneName=dyn_tail_base}, scale=<0 0 0>}
   Output Pose(Result=Component To Local(ComponentPose=@n8)) {layerGroup=None}
   n9 = Control Rig(Source=Copy Pose From Mesh() {bUseAttachedParent=true}) {controlRigAssetReference {blueprintRigClass=&/F_MED_FutureRabbit_CtrlRig.F_MED_FutureRabbit_CtrlRig_C}}
   Save cached pose 'Deform_1'(Pose=@n9) {cacheName=Deform_1}
  ]}
 EventGraph {nodes [
   Event BlueprintUpdateAnimation()
   > Sequence()
   > n1 = "Gravity Override State Function (Parameter Asset) (Nativized)"(
    CharacterPartAnimInstance=Self-Reference()
    GameTimeInSeconds=GetGameTimeInSeconds()
    Parameters=&/F_MED_FutereRabbit_Parameters
   )
   > Set Gravity Override(Gravity Override=@n1)
   n2 = CheckForBackbling()
   > Set isWearingPack(isWearingPack:bool=OR Boolean!(A=@n2, B=@n2.bIsBackblingACape, C:bool=@n2.bIsBackblingLong) -> (ReturnValue)) -> (Output_Get:bool)
   > Set isWearingCape(isWearingCape=@n2.bIsBackblingLong)
   n3 = GetCurrentActiveMontage!(self:Anim Instance Object Reference=GetOwnerAnimInstance!(self:Custom Character Part Anim Instance Object Reference) -> (ReturnValue:Anim Instance Object Reference)) -> (ReturnValue:Anim Montage Object Reference)
   > "Set scale tail  cauldro"("scale tail  cauldro":bool="Equal (Object)"(A=@n3, B=/BRCosmetics/Animation/Game/MainPlayer/Emotes/Cauldron/Emote_Cauldron_CMF_M)) -> (Output_Get:bool)
   > "Set scale tail  scrubadub"("scale tail  scrubadub":bool="Equal (Object)"(A=@n3, B=/BRCosmetics/Animation/Game/MainPlayer/Emotes/Scrub_A_Dub/Emote_ScrubaDub_CMF_M)) -> (Output_Get:bool)
   GetPlayerPawn!(PlayerIndex:int=0) -> (ReturnValue:Pawn Object Reference)
  ]}
]
variables [
 Character Mesh {type=SkeletalMeshComponent}
 disableHalfDynamics {type=boolean}
 Gravity Override {type=Vector}
 IScelebration {type=boolean}
 isWearingCape {type=boolean}
 isWearingPack {type=boolean}
 "scale tail  cauldro" {type=boolean}
 "scale tail  scrubadub" {type=boolean}
]
