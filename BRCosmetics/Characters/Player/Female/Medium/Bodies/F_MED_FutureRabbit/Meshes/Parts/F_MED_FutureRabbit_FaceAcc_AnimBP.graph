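# Post-process style secondary-motion AnimBP for the FutureRabbit female face accessory mesh, adding physics sway to its hair bones.
# It copies the attached parent's pose, then chains twelve root-relative Anim Dynamics nodes for side, back and front hair.
# Each node uses an angular spring, gravity override and per-bone angular limits.
# The event graph smooths AlphaJump from jump/fall state, but the AnimGraph never reads it; every node runs at full alpha.
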AnimBlueprint /BRCosmetics/Characters/Player/Female/Medium/Bodies/F_MED_FutureRabbit/Meshes/Parts/F_MED_FutureRabbit_FaceAcc_AnimBP
target_skeleton=&/F_MED_FutureRabbit_FaceAcc_Skeleton
graphs [
 AnimGraph {nodes [
   n1 = Anim Dynamics - Bone: dyn_hair_l_01(ComponentPose=Local To Component(LocalPose=Copy Pose From Mesh() {bUseAttachedParent=true}), Alpha=1, ExternalForce="0,0,0") {
    angularDampingOverride=1
    angularSpringConstant=200
    bAngularSpring=true
    bOverrideAngularDamping=true
    bUseGravityOverride=true
    simulationSpace=RootRelative
    boundBone {boneName=dyn_hair_l_01}
    componentAppliedLinearAccClamp=<500 500 500>
    componentLinearAccScale=<0.1 0.1 0.3>
    componentLinearVelScale=<0.1 0.1 0.3>
    gravityOverride {z=-980}
    physicsBodyDefinitions [{
      boundBone {boneName=dyn_hair_l_01}
      boxExtents=<21 21 21>
      constraintSetup {angularLimitsMax=<18 6 33>, angularLimitsMin=<-20 1 -8>, angularTarget=<1 0 0>}
      localJointOffset=<1 0 0.3>
     }]
    relativeSpaceBone {boneName=head}
    simSpaceSettings {maxAngularAcceleration=35, maxAngularVelocity=7, simSpaceAngularAlpha=1}
   }
   n2 = Anim Dynamics - Bone: dyn_hair_l_02(ComponentPose=@n1, Alpha=1, ExternalForce="0,0,0") {
    angularDampingOverride=1
    angularSpringConstant=200
    bAngularSpring=true
    bOverrideAngularDamping=true
    bUseGravityOverride=true
    simulationSpace=RootRelative
    boundBone {boneName=dyn_hair_l_02}
    componentAppliedLinearAccClamp=<500 500 500>
    componentLinearAccScale=<0.1 0.1 0.3>
    componentLinearVelScale=<0.1 0.1 0.3>
    gravityOverride {z=-980}
    physicsBodyDefinitions [{
      boundBone {boneName=dyn_hair_l_02}
      boxExtents=<9 9 9>
      constraintSetup {angularLimitsMax=<18 7 15>, angularLimitsMin=<-12 -11 -11>, angularTarget=<1 0 0>}
      localJointOffset=<1 0 0.3>
     }]
    relativeSpaceBone {boneName=head}
    simSpaceSettings {maxAngularAcceleration=35, maxAngularVelocity=7, simSpaceAngularAlpha=1}
   }
   n3 = Anim Dynamics - Bone: dyn_hair_r_01(ComponentPose=@n2, Alpha=1, ExternalForce="0,0,0") {
    angularDampingOverride=1
    angularSpringConstant=200
    bAngularSpring=true
    bOverrideAngularDamping=true
    bUseGravityOverride=true
    simulationSpace=RootRelative
    boundBone {boneName=dyn_hair_r_01}
    componentAppliedLinearAccClamp=<500 500 500>
    componentLinearAccScale=<0.1 0.1 0.3>
    componentLinearVelScale=<0.1 0.1 0.3>
    gravityOverride {z=-980}
    physicsBodyDefinitions [{
      boundBone {boneName=dyn_hair_r_01}
      boxExtents=<12 12 12>
      constraintSetup {angularLimitsMax=<12 7 4>, angularLimitsMin=<-20 2 -33>, angularTarget=<1 0 0>}
      localJointOffset=<1 0 0.3>
     }]
    relativeSpaceBone {boneName=head}
    simSpaceSettings {maxAngularAcceleration=35, maxAngularVelocity=7, simSpaceAngularAlpha=1}
   }
   n4 = Anim Dynamics - Bone: dyn_hair_r_02(ComponentPose=@n3, Alpha=1, ExternalForce="0,0,0") {
    angularDampingOverride=1
    angularSpringConstant=200
    bAngularSpring=true
    bOverrideAngularDamping=true
    bUseGravityOverride=true
    simulationSpace=RootRelative
    boundBone {boneName=dyn_hair_r_02}
    componentAppliedLinearAccClamp=<500 500 500>
    componentLinearAccScale=<0.1 0.1 0.3>
    componentLinearVelScale=<0.1 0.1 0.3>
    gravityOverride {z=-980}
    physicsBodyDefinitions [{
      boundBone {boneName=dyn_hair_r_02}
      boxExtents=<12 12 12>
      constraintSetup {angularLimitsMax=<12 22 5>, angularLimitsMin=<-12 -11 -15>, angularTarget=<1 0 0>}
      localJointOffset=<1 0 0.3>
     }]
    relativeSpaceBone {boneName=head}
    simSpaceSettings {maxAngularAcceleration=35, maxAngularVelocity=7, simSpaceAngularAlpha=1}
   }
   n5 = Anim Dynamics - Bone: dyn_hair_bk_l_01(ComponentPose=@n4, Alpha=1, ExternalForce="0,0,0") {
    angularDampingOverride=0.999
    angularSpringConstant=200
    bAngularSpring=true
    bOverrideAngularDamping=true
    bUseGravityOverride=true
    simulationSpace=RootRelative
    boundBone {boneName=dyn_hair_bk_l_01}
    componentAppliedLinearAccClamp=<3000 3000 3000>
    componentLinearAccScale=<0.1 0.1 0.4>
    componentLinearVelScale=<0.1 0.1 0.4>
    gravityOverride {z=-980}
    physicsBodyDefinitions [{
      boundBone {boneName=dyn_hair_bk_l_01}
      boxExtents=<8 8 8>
      constraintSetup {angularLimitsMax=<9 22 15>, angularLimitsMin=<-9 -22 -15>, angularTarget=<1 0 0>}
      localJointOffset=<1 0.3 0>
     }]
    relativeSpaceBone {boneName=spine_05}
    simSpaceSettings {maxAngularAcceleration=40, maxAngularVelocity=8, simSpaceAngularAlpha=0.5}
   }
   n6 = Anim Dynamics - Bone: dyn_hair_bk_r_01(ComponentPose=@n5, Alpha=1, ExternalForce="0,0,0") {
    angularDampingOverride=0.9999
    angularSpringConstant=300
    bAngularSpring=true
    bOverrideAngularDamping=true
    bUseGravityOverride=true
    simulationSpace=RootRelative
    boundBone {boneName=dyn_hair_bk_r_01}
    componentAppliedLinearAccClamp=<3000 3000 3000>
    componentLinearAccScale=<0.1 0.1 0.4>
    componentLinearVelScale=<0.1 0.1 0.4>
    gravityOverride {z=-980}
    physicsBodyDefinitions [{
      boundBone {boneName=dyn_hair_bk_r_01}
      boxExtents=<9 9 9>
      constraintSetup {angularLimitsMax=<9 22 15>, angularLimitsMin=<-9 -22 -15>, angularTarget=<1 0 0>}
      localJointOffset=<1 0 0.5>
     }]
    relativeSpaceBone {boneName=spine_05}
    simSpaceSettings {maxAngularAcceleration=40, maxAngularVelocity=8, simSpaceAngularAlpha=0.5}
   }
   n7 = Anim Dynamics - Bone: dyn_hair_right_1_01(ComponentPose=@n6, Alpha=1, ExternalForce="0,0,0") {
    angularDampingOverride=0.9999
    angularSpringConstant=300
    bAngularSpring=true
    bOverrideAngularDamping=true
    bUseGravityOverride=true
    simulationSpace=RootRelative
    boundBone {boneName=dyn_hair_right_1_01}
    componentAppliedLinearAccClamp=<3000 3000 3000>
    componentLinearAccScale=<0.1 0.1 0.4>
    componentLinearVelScale=<0.1 0.1 0.4>
    gravityOverride {z=-980}
    physicsBodyDefinitions [{
      boundBone {boneName=dyn_hair_right_1_01}
      boxExtents=<13 13 13>
      constraintSetup {angularLimitsMax=<9 22 15>, angularLimitsMin=<-9 -22 -15>, angularTarget=<1 0 0>}
      localJointOffset=<1 0 0.5>
     }]
    relativeSpaceBone {boneName=spine_05}
    simSpaceSettings {maxAngularAcceleration=40, maxAngularVelocity=8, simSpaceAngularAlpha=0.5}
   }
   n8 = Anim Dynamics - Bone: dyn_hair_right_01(ComponentPose=@n7, Alpha=1, ExternalForce="0,0,0") {
    angularDampingOverride=0.9999
    angularSpringConstant=300
    bAngularSpring=true
    bOverrideAngularDamping=true
    bUseGravityOverride=true
    simulationSpace=RootRelative
    boundBone {boneName=dyn_hair_right_01}
    componentAppliedLinearAccClamp=<3000 3000 3000>
    componentLinearAccScale=<0.1 0.1 0.4>
    componentLinearVelScale=<0.1 0.1 0.4>
    gravityOverride {z=-980}
    physicsBodyDefinitions [{
      boundBone {boneName=dyn_hair_right_01}
      boxExtents=<10 10 10>
      constraintSetup {angularLimitsMax=<9 22 15>, angularLimitsMin=<-9 -22 -15>, angularTarget=<1 0 0>}
      localJointOffset=<1 0 0.5>
     }]
    relativeSpaceBone {boneName=spine_05}
    simSpaceSettings {maxAngularAcceleration=40, maxAngularVelocity=8, simSpaceAngularAlpha=0.5}
   }
   n9 = Anim Dynamics - Bone: dyn_hair_left_1_01(ComponentPose=@n8, Alpha=1, ExternalForce="0,0,0") {
    angularDampingOverride=0.999
    angularSpringConstant=200
    bAngularSpring=true
    bOverrideAngularDamping=true
    bUseGravityOverride=true
    simulationSpace=RootRelative
    boundBone {boneName=dyn_hair_left_1_01}
    componentAppliedLinearAccClamp=<3000 3000 3000>
    componentLinearAccScale=<0.1 0.1 0.4>
    componentLinearVelScale=<0.1 0.1 0.4>
    gravityOverride {z=-980}
    physicsBodyDefinitions [{
      boundBone {boneName=dyn_hair_left_1_01}
      boxExtents=<12 12 12>
      constraintSetup {angularLimitsMax=<9 22 15>, angularLimitsMin=<-9 -22 -15>, angularTarget=<1 0 0>}
      localJointOffset=<1 0.3 0>
     }]
    relativeSpaceBone {boneName=spine_05}
    simSpaceSettings {maxAngularAcceleration=40, maxAngularVelocity=8, simSpaceAngularAlpha=0.5}
   }
   n10 = Anim Dynamics - Bone: dyn_hair_left_01(ComponentPose=@n9, Alpha=1, ExternalForce="0,0,0") {
    angularDampingOverride=0.999
    angularSpringConstant=200
    bAngularSpring=true
    bOverrideAngularDamping=true
    bUseGravityOverride=true
    simulationSpace=RootRelative
    boundBone {boneName=dyn_hair_left_01}
    componentAppliedLinearAccClamp=<3000 3000 3000>
    componentLinearAccScale=<0.1 0.1 0.4>
    componentLinearVelScale=<0.1 0.1 0.4>
    gravityOverride {z=-980}
    physicsBodyDefinitions [{
      boundBone {boneName=dyn_hair_left_01}
      boxExtents=<14 14 14>
      constraintSetup {angularLimitsMax=<9 22 15>, angularLimitsMin=<-9 -22 -15>, angularTarget=<1 0 0>}
      localJointOffset=<1 0.3 0>
     }]
    relativeSpaceBone {boneName=spine_05}
    simSpaceSettings {maxAngularAcceleration=40, maxAngularVelocity=8, simSpaceAngularAlpha=0.5}
   }
   n11 = Anim Dynamics - Bone: dyn_hair_front_1_l_base(ComponentPose=@n10, Alpha=1, ExternalForce="0,0,0") {
    angularDampingOverride=0.999
    angularSpringConstant=200
    bAngularSpring=true
    bOverrideAngularDamping=true
    bUseGravityOverride=true
    simulationSpace=RootRelative
    boundBone {boneName=dyn_hair_front_1_l_base}
    componentAppliedLinearAccClamp=<3000 3000 3000>
    componentLinearAccScale=<0.1 0.1 0.4>
    componentLinearVelScale=<0.1 0.1 0.4>
    gravityOverride {z=-980}
    physicsBodyDefinitions [{
      boundBone {boneName=dyn_hair_front_1_l_base}
      boxExtents=<12 12 12>
      constraintSetup {angularLimitsMax=<9 22 22>, angularLimitsMin=<-9 -22 -1>, angularTarget=<1 0 0>}
      localJointOffset=<1 1 0>
     }]
    relativeSpaceBone {boneName=spine_05}
    simSpaceSettings {maxAngularAcceleration=40, maxAngularVelocity=8, simSpaceAngularAlpha=0.5}
   }
   n12 = Anim Dynamics - Bone: dyn_hair_front_base(ComponentPose=@n11, Alpha=1, ExternalForce="0,0,0") {
    angularDampingOverride=0.999
    angularSpringConstant=200
    bAngularSpring=true
    bOverrideAngularDamping=true
    bUseGravityOverride=true
    simulationSpace=RootRelative
    boundBone {boneName=dyn_hair_front_base}
    componentAppliedLinearAccClamp=<3000 3000 3000>
    componentLinearAccScale=<0.1 0.1 0.4>
    componentLinearVelScale=<0.1 0.1 0.4>
    gravityOverride {z=-980}
    physicsBodyDefinitions [{
      boundBone {boneName=dyn_hair_front_base}
      boxExtents=<14 14 14>
      constraintSetup {angularLimitsMax=<9 22 22>, angularLimitsMin=<-9 -22 -1>, angularTarget=<1 0 0>}
      localJointOffset=<1 1 0>
     }]
    relativeSpaceBone {boneName=spine_05}
    simSpaceSettings {maxAngularAcceleration=40, maxAngularVelocity=8, simSpaceAngularAlpha=0.5}
   }
   Output Pose(Result=Component To Local(ComponentPose=@n12)) {layerGroup=None}
  ]}
 EventGraph {nodes [
   n1 = Event BlueprintUpdateAnimation()
   > Set Delta Time X(Delta Time X=@n1.DeltaTimeX)
   n2 = OR Boolean(A=Get bIsJumping(), B=Get bIsFalling())
   > Set AlphaJump(AlphaJump:double=FInterpTo(Current=Get AlphaJump() -> (AlphaJump:double), Target=SelectFloat(A=0.2, B=0.7, bPickA=@n2), DeltaTime=Get Delta Time X(), InterpSpeed=5)) -> (Output_Get:double)
  ]}
]
variables [
 AlphaJump {type=number}
 Delta Time X {type=number}
]
